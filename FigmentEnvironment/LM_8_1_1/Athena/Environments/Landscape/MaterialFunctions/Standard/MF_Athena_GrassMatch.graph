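# MF_Athena_GrassMatch is a landscape material function that tints an input BaseColor by the terrain ground color map, likely so grass matches the ground beneath it.
# It samples the GroundColor texture (linear color) at world-position-based UVs from MF_Athena_Terrain_Coords.
# The input is multiplied by that RGB, then scaled by the ColormapDivisor parameter from the MPC_Landscape collection.
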MaterialFunction /FigmentEnvironment/LM_8_1_1/Athena/Environments/Landscape/MaterialFunctions/Standard/MF_Athena_GrassMatch
graph {nodes [
  n1 = FunctionInput() {inputName=BaseColor}
  n2 = TextureSampleParameter2D(UVs=MaterialFunctionCall!(WorldPosition) {materialFunction=/FigmentEnvironment/LM_8_1_1/Athena/Environments/Landscape/MaterialFunctions/Utilities/MF_Athena_Terrain_Coords}.Result) {
   parameterName=GroundColor
   samplerType=SAMPLERTYPE_LinearColor
   texture=/FigmentEnvironment/LM_8_1_1/Athena/Environments/Landscape/Textures/T_Athena_Terrain_GroundColor
  }
  FunctionOutput(None=Multiply(
    A=Multiply(A=@n1, B=@n2.RGB)
    B=CollectionParameter() {collection=/FigmentEnvironment/LM_8_1_1/Athena/Environments/Landscape/MPC/MPC_Landscape, parameterName=ColormapDivisor}
   ))
 ]}
