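# Landscape material function that builds a gravel surface layer for the Athena environment and returns it as Material Attributes.
# It samples a white rock diffuse texture at the Coords input with a shared wrap sampler, then multiplies RGB by the Color Tint input for BaseColor.
# Texture alpha serves as Height, exposed separately alongside BaseColor and also packed into Opacity.
# Specular is a low constant, Roughness is fully rough, and no normal map is used.
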MaterialFunction /FigmentEnvironment/LM_8_1_1/Athena/Environments/Landscape/MaterialFunctions/Standard/MF_Athena_Gravel_01
graph {nodes [
  n1 = TextureSample(UVs=FunctionInput(Preview=TextureCoordinate()) {bUsePreviewValueAsDefault=true, inputName=Coords, inputType=FunctionInput_Vector2}) {samplerSource=SSM_Wrap_WorldGroupSettings, texture=/FigmentEnvironment/LM_8_1_1/Packages/DS_Fortnight/WhiteRockCustom_D}
  n2 = Multiply(
   A=@n1.RGB
   B=FunctionInput(Preview=Constant() {r=0.8}) {bUsePreviewValueAsDefault=true, inputName=Color Tint, sortPriority=1, previewValue=<1 1 1 1>}
  )
  FunctionOutput(None=MakeMaterialAttributes!(
    BaseColor=FunctionOutput(None=@n2) {outputName=BaseColor, sortPriority=2}
    Metallic
    Specular=Constant() {r=0.2}
    Roughness=Constant() {r=1}
    Anisotropy
    EmissiveColor
    Opacity=FunctionOutput(None=@n1.A) {outputName=Height, sortPriority=1}
    OpacityMask
    Normal
    Tangent
    WorldPositionOffset
    SubsurfaceColor
    ClearCoat
    ClearCoatRoughness
    AmbientOcclusion
    Refraction
    CustomizedUVs_0
    CustomizedUVs_1
    CustomizedUVs_2
    CustomizedUVs_3
    CustomizedUVs_4
    CustomizedUVs_5
    CustomizedUVs_6
    CustomizedUVs_7
    PixelDepthOffset
    ShadingModel
    Displacement
   ))
 ]}
