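# Athena landscape rock layer function outputting material attributes, base color, height and projection contrast.
# World-aligned cave-wall textures blend by slope: large-scale rock with normals on steep faces, darkened smaller-scale rock with flat normals on shallower ones, Gravel input on the flattest.
# Medium quality blends one rock sample with gravel.
# Low quality and mobile color use only that sample, choosing XZ or YZ projection with explicit derivatives.
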MaterialFunction /FigmentEnvironment/LM_8_1_1/Athena/Environments/Landscape/MaterialFunctions/Standard/MF_Athena_Rock_01
graph {nodes [
  n1 = Constant() {r=10}
  n2 = MaterialFunctionCall() {materialFunction=&/MF_Athena_RoadGravel}
  n3 = FunctionInput(Preview=AppendVector(A=@n2.Gravel, B=@n2.Gravel)) {bUsePreviewValueAsDefault=true, inputName=Gravel, inputType=FunctionInput_Vector4}
  n4 = Multiply(A=VertexNormalWS(), B=Constant() {r=0.725})
  n5 = Abs(None=ComponentMask(None=@n4) {b=true})
  n6 = LinearInterpolate(
   A=MaterialFunctionCall(
    TextureObject=TextureObject() {texture=/FigmentEnvironment/LM_8_1_1/Packages/DS_Fortnite_Terrain_NoLOD/Textures/T_Cave_Exterior_Wall_D}
    TextureSize=Constant() {r=1534.9661}
    Export Float 4=StaticBool() {value=true}
    ProjectionTransitionContrast=@n1
   ) {materialFunction=&WorldAlignedTexture}.XY Texture
   B=Multiply(A=@n3, B=Constant() {r=1.75})
   Alpha=ComponentMask(None=MaterialFunctionCall(In=@n5, Contrast=@n1) {materialFunction=&CheapContrast}.Result) {r=true}
  )
  n7 = TextureObject() {texture=&T_Cave_Exterior_Wall_D}
  n8 = Constant() {r=4500}
  n9 = StaticBool() {value=true}
  n10 = Constant() {r=5}
  n11 = Constant() {r=-4}
  n12 = ComponentMask(None=VertexNormalWS()) {b=true}
  n13 = Clamp(None=Add(A=LinearInterpolate(A=@n10, B=@n11, Alpha=@n12), B=Constant() {r=0.8}))
  n14 = Divide(A=LinearInterpolate(
    A=Divide(A=@n6) {constB=1.25, desc=Darken Shallow rock to make up for no normal}
    B=MaterialFunctionCall(TextureObject=@n7, TextureSize=@n8, Export Float 4=@n9, ProjectionTransitionContrast=@n1) {materialFunction=&WorldAlignedTexture}.XY Texture
    Alpha=@n13
   )) {constB=1.5}
  n15 = AppendVector(A=Constant() {r=19.716}, B=Constant() {r=43.9378})
  n16 = Divide(A=LandscapeLayerCoords() {mappingScale=1, mappingType=TCMT_XZ}, B=@n15)
  n17 = Divide(A=LandscapeLayerCoords() {mappingScale=1, mappingType=TCMT_YZ}, B=@n15)
  n18 = Abs(None=VertexNormalWS())
  n19 = Ceil(None=Subtract(A=ComponentMask(None=@n18) {r=true}, B=ComponentMask(None=@n18) {g=true}))
  n20 = DDX(Value=@n16)
  n21 = DDY(Value=@n16)
  n22 = DDX(Value=@n17)
  n23 = DDY(Value=@n17)
  n24 = Max(A=Abs(None=AppendVector(A=@n20, B=@n21)), B=Abs(None=AppendVector(A=@n22, B=@n23)))
  n25 = TextureSample!(
   UVs=LinearInterpolate(A=@n16, B=@n17, Alpha=Saturate(None=@n19))
   Tex
   "DDX(UVs)"=ComponentMask(None=@n24) {g=true, r=true}
   "DDY(UVs)"=ComponentMask(None=@n24) {a=true, b=true}
   Apply View MipBias
  ) {
   mipValueMode=TMVM_Derivative
   samplerSource=SSM_Wrap_WorldGroupSettings
   texture=&T_Cave_Exterior_Wall_D
  }
  n26 = Divide(A=AppendVector(A=@n25.RGB, B=@n25.RGB))
  n27 = MaterialFunctionCall(In=Abs(None=ComponentMask(None=@n4) {b=true}), Contrast=@n1) {materialFunction=&CheapContrast}
  n28 = QualitySwitch(
   Default=@n14
   Low=@n26
   High=@n14
   Medium=LinearInterpolate(A=@n26, B=Divide(A=@n3) {constB=1.25}, Alpha=ComponentMask(None=@n27.Result) {r=true})
  ) {desc=Use only one rock blend on Medium and Low}
  n29 = FeatureLevelSwitch(Default=@n28, ES2=@n26, ES3_1=@n26, SM4_REMOVED=@n28, SM5=@n28, SM6=@n28)
  n30 = ComponentMask(None=@n29) {b=true, g=true, r=true}
  n31 = TextureObject() {samplerType=SAMPLERTYPE_Normal, texture=/FigmentEnvironment/LM_8_1_1/Packages/DS_Fortnite_Terrain_NoLOD/Textures/T_Cave_Exterior_Wall_N}
  n32 = MaterialFunctionCall!(
   TextureObject=@n31
   TextureSize=@n8
   WorldSpace
   Normal
   ProjectionTransitionContrast=Constant() {r=-10}
   WorldPosition
   Use High Quality Normals
  ) {materialFunction=/FigmentEnvironment/LM_8_1_1/Athena/Environments/Materials/WorldAlignedNormalContrast}
  n33 = LinearInterpolate(A=Constant3Vector() {constant=<rgba 0 0 1 0>}, B=@n32.XY Texture, Alpha=@n13)
  n34 = ComponentMask(None=@n29) {a=true}
  FunctionOutput(None=MakeMaterialAttributes!(
    BaseColor=@n30
    Metallic
    Specular=ScalarParameter() {defaultValue=0.4, parameterName=IceRockSpecular}
    Roughness=Constant() {r=0.9}
    Anisotropy
    EmissiveColor
    Opacity=@n34
    OpacityMask
    Normal=QualitySwitch(Default=@n33, Low=@n32.XY Texture, High=@n33, Medium=@n32.XY Texture) {desc=Use only one rock blend on Medium and Low}
    Tangent
    WorldPositionOffset
    SubsurfaceColor
    ClearCoat
    ClearCoatRoughness
    AmbientOcclusion
    Refraction
    CustomizedUVs_0
    CustomizedUVs_1
    CustomizedUVs_2
    CustomizedUVs_3
    CustomizedUVs_4
    CustomizedUVs_5
    CustomizedUVs_6
    CustomizedUVs_7
    PixelDepthOffset
    ShadingModel
    Displacement
   ))
  FunctionOutput(None=@n30) {outputName=Rock_BaseColor}
  FunctionOutput(None=@n34) {outputName=Height}
  FunctionOutput(None=@n1) {outputName=Proj_Contrast}
 ]}
